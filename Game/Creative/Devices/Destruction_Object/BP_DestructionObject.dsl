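; Fortnite Creative destruction objective device: a team-owned, damageable object with selectable mesh, beacon, health bar and optional explosion.
; Meshes, particles and widget load asynchronously; local team or spectator changes refresh the beacon.
; Damage requires a running minigame and a different team; health changes drive effects, tracked progress and feedback events.
; Destruction awards score, broadcasts an event and can run a throttled async sphere overlap that destroys nearby actors, chaining into other objectives.
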(event EditorPreview|EventSetupForEditorPreview
  (CallFunction|PickMesh)
  (Audio|Parameter|SetIntegerParameter (Variables|Default|GetDeath%20Sound) "ObjectIndex" (Variables|Default|GetMeshIndex) :!class /Script/AudioExtensions.AudioParameterControllerInterface)
  (CallFunction|SetParticle)
  (Health|BindEventtoOnActorHealthChanged (delegate Custom|HealthChange)))

(event EventBeginPlay
  (Utilities|FlowControl|Sequence
    (:then_0
      (CallFunction|PickMesh)
      (Audio|Parameter|SetIntegerParameter (Variables|Default|GetDeath%20Sound) "ObjectIndex" (Variables|Default|GetMeshIndex) :!class /Script/AudioExtensions.AudioParameterControllerInterface)
      (CallFunction|SetParticle)
      (Health|BindEventtoOnActorHealthChanged (delegate Custom|HealthChange)))
    (:then_1
      (Networking|CanExecuteCosmeticEvents
        (:True
          (Utilities|FlowControl|Sequence
            (:then_0
              ;[0 0 0] client only-create the hp bar ui
              (bind _class (Utilities|AsyncLoadClassAsset (Variables|UI|GetHealthBarSoftClassRef))
                (:Completed
                  (bind _asuser_widget (Utilities|Casting|CastToUserWidgetClass _class))
                  (bind _returnvalue (UserInterface|CreateWidget _asuser_widget (Game|GetPlayerController 0)))
                  (UserInterface|SetWidget (Variables|Default|GetHealthBarComp) _returnvalue)
                  (Init|InitfromObject(Message) _returnvalue self))))
            (:then_1
              (for _array_element (Fortnite|GetLocalFortPlayerControllers false true)
                (bind _asfort_player_controller_spectating (Utilities|Casting|CastToFortPlayerControllerSpectating _array_element))
                (EventDispatchers|BindEventtoOnFollowedPlayerChangedTeam _asfort_player_controller_spectating (EventDispatchers|CreateEvent :!event Custom|OnLocalPlayerChangedTeams))
                (EventDispatchers|BindEventtoOnCameraTypeChanged _asfort_player_controller_spectating (EventDispatchers|CreateEvent :!event Custom|OnSpectatorCameraTypeChanged))
                (CallFunction|OnLocalPlayerChangedTeams)))))))
    (:then_2
      (CallFunction|SetShouldTickHealthBarComponent))))

(event Custom|SetVisibilityForcedDuringGame
  (params (NewForcedValue :type Enum_ForcedValueDuringGame%20Enum :default "NewEnumerator0"))
  ;[0.25 0.25 0.25] ForcedValuesDuringGame
  (Networking|SwitchHasAuthority
    (:Authority
      (if (Utilities|Enum|Equal(Enum) (Variables|UserKnobs|GetCollisionDuringGamesKnob) "NewEnumerator3")
        (CallFunction|SetCollisionForcedDuringGame NewForcedValue)
        (else
          _))
      (Visibility|SetForcedVisibilityDuringGame (Variables|Default|GetVisibleInGameComponent) NewForcedValue))))

(event Custom|HealthChange
  (params (Actor :type Actor%20Object%20Reference) (NewHealth :type Float%20%28single-precision%29))
  (CallFunction|UpdateTrackedProgress)
  (if (CallFunction|ShouldShowFX)
    (CallFunction|UpdateHealthEffects NewHealth)))

(event EventOnWorldReady
  (Visibility|SetVisibilityComponents (Variables|Default|GetVisibleInGameComponent) (Utilities|Array|MakeArray (Variables|Debug|GetStaticMeshComponent)) true))

(event Custom|SetCollisionForcedDuringGame
  (params (NewForcedValue :type Enum_ForcedValueDuringGame%20Enum :default "NewEnumerator0"))
  ;[0.25 0.25 0.25] ForcedValuesDuringGame
  (Networking|SwitchHasAuthority
    (:Authority
      (if (Utilities|Enum|NotEqual(Enum) NewForcedValue (Variables|Replicated|GetCollisionForcedDuringGame))
        (Variables|Replicated|SetCollisionForcedDuringGame NewForcedValue)))))

(event Custom|Update%20Healthbar%20Height
  (bind (_min _max) (Utilities|Struct|BreakBox (StaticMesh|GetBoundingBox (Variables|/Script/Engine.StaticMeshComponent|GetStaticMesh (Variables|Debug|GetStaticMeshComponent)))))
  (bind _staticmeshcomponent (Variables|Debug|GetStaticMeshComponent))
  (bind _staticmesh (Variables|/Script/Engine.StaticMeshComponent|GetStaticMesh _staticmeshcomponent))
  ;[0 0 0] Healthbar, update content and visibility
  (Utilities|IsValid _staticmeshcomponent
    (:Is%20Valid
      (Utilities|IsValid _staticmesh
        (:Is%20Valid
          (Transformation|SetRelativeLocation (Variables|Default|GetHealthBarComp) (Math|Vector|MakeVector 0.0 0.0 (.z _max)) false false))))))

(event MinigameStateChangedDelegate(FortMinigameLogic)
  (params (Minigame :type Fort%20Minigame%20Object%20Reference) (NewMinigameState :type EFortMinigameState%20Enum :default "PreGame"))
  ;[0.5 0.25 0.5] Minigame State Changes
  (if (Utilities|Enum|Equal(Enum) NewMinigameState "PostGameReset")
    (Variables|Replicated|SetCollisionForcedDuringGame "NewEnumerator0")
    (Default|UpdateStaticMeshCollision)))

(event Custom|OnLocalPlayerChangedTeams
  (CallFunction|SetParticle)
  (Default|CallLocalPlayerChangedTeams))

(event Custom|OnSpectatorCameraTypeChanged
  (params (SpectatingPC :type Fort%20Player%20Controller%20Spectating%20Object%20Reference) (CameraType :type ESpectatorCameraType%20Enum :default "ThirdPerson"))
  (CallFunction|SetParticle)
  (Default|CallLocalPlayerChangedTeams))

(event Custom|Turn%20on%20Visibility%20Event
  (params (Instigator :type Controller%20Object%20Reference))
  ;[0.2393 0 1 0.25] Gameplay Functions
  (CallFunction|SetVisibilityForcedDuringGame "NewEnumerator1"))

(event OnVisibilityChanged(VisibleInGameComponent)
  (params (Visible :type Boolean))
  (Default|UpdateStaticMeshCollision))

(event Custom|Turn%20off%20Visibility%20Event
  (params (Instigator :type Controller%20Object%20Reference))
  ;[0.2393 0 1 0.25] Gameplay Functions
  (CallFunction|SetVisibilityForcedDuringGame "NewEnumerator2"))

(event Custom|StopHealthEffects
  (Components|Activation|SetActive (Variables|Default|GetDamageParticleSystem) false true))

(event Custom|RPC%20Play%20Death%20Sound
  (Audio|Components|Audio|Play (Variables|Default|GetDeath%20Sound) 0.0)
  (Networking|CanExecuteCosmeticEvents
    (:True
      (if (not (<= (Variables|Default|GetExplosionRadius) 0))
        (Effects|SpawnEmitteratLocation :EmitterTemplate "/Game/Creative/Devices/ExplosiveBarrel/P_ExplosiveBarrel_ExplosionMain.P_ExplosiveBarrel_ExplosionMain" :Location (Transformation|GetWorldLocation (Variables|Debug|GetStaticMeshComponent)) :Scale (CallFunction|GetExplosionFXScale) :bAutoDestroy true :PoolingMethod "AutoRelease" :bAutoActivateSystem true)))))

(event Custom|Activate%20Objective%20Trail
  (params (Instigator :type Controller%20Object%20Reference))
  ;[0.2393 0 1 0.25] Gameplay Functions
  ;;[0.2393 0 1 0.25] ObjectiveTrail
  (Variables|Default|SetObjectiveTrailInstigator Instigator)
  (Variables|Default|SetActivateObjectiveTrail true)
  (Class|CreativeObjectiveTrailComponent|ToggleObjectiveTrail (Variables|Default|GetCreative_ObjectiveTrail_Component) (Variables|Default|GetActivateObjectiveTrail) self (Variables|Default|GetObjectiveTrailInstigator) (Variables|Default|GetTeamSetting) (Variables|Default|GetKnob_bShowObjectiveTrailFriendly) (Variables|Default|GetKnob_bShowObjectiveTrailEnemy) (Variables|Default|GetKnob_bShowObjectiveTrailInstigator) false 2.0 0))

(event EventOnDamagePlayEffects
  (params (Damage :type Float%20%28single-precision%29) (DamageTags :type Gameplay%20Tag%20Container%20Structure%20%28by%20ref%29) (Momentum :type Vector :default "0, 0, 0") (HitInfo :type Hit%20Result%20Structure%20%28by%20ref%29) (InstigatedBy :type Fort%20Pawn%20Object%20Reference) (DamageCauser :type Actor%20Object%20Reference) (EffectContext :type Gameplay%20Effect%20Context%20Handle%20Structure))
  ;[1 1 1] Destruction
  (if (Utilities|Enum|Equal(Enum) (Variables|UserKnobs|GetPlayAudioVFXKnob) "NewEnumerator1")
    (Audio|Components|Audio|Play (Variables|Default|GetImpact%20Sound) 0.0)))

(event Custom|Deactivate%20Objective%20Trail
  (params (Instigator :type Controller%20Object%20Reference))
  ;[0.2393 0 1 0.25] Gameplay Functions
  ;;[0.2393 0 1 0.25] ObjectiveTrail
  (Variables|Default|SetObjectiveTrailInstigator Instigator)
  (Variables|Default|SetActivateObjectiveTrail false)
  (Class|CreativeObjectiveTrailComponent|ToggleObjectiveTrail (Variables|Default|GetCreative_ObjectiveTrail_Component) (Variables|Default|GetActivateObjectiveTrail) self (Variables|Default|GetObjectiveTrailInstigator) (Variables|Default|GetTeamSetting) (Variables|Default|GetKnob_bShowObjectiveTrailFriendly) (Variables|Default|GetKnob_bShowObjectiveTrailEnemy) (Variables|Default|GetKnob_bShowObjectiveTrailInstigator) false 2.0 0))

(event Custom|Play50%25Effect
  (bind _damageparticlesystem (Variables|Default|GetDamageParticleSystem))
  (Components|Activation|SetActive _damageparticlesystem true false)
  (Effects|Components|ParticleSystem|SetEmitterEnable _damageparticlesystem "Sparks" false))

(event Custom|Destroy%20Event
  (params (Instigator :type Controller%20Object%20Reference))
  ;[1 1 1] Destruction
  (Variables|Default|SetDestroyInstigator Instigator)
  (if (not (Variables|BuildingActor|GetbDestroyed))
    (CallFunction|DoDestroy)
    (CallFunction|HandleHealthChangeEvents 0.0 (Variables|Default|GetDestroyInstigator) true)
    (Minigame|Objective|SetTrackedProgress (Variables|Default|GetFortMinigameObjective) 0.0)
    (Actor|DestroyActor)
    (GameplayEvent|BroadcastEvent (Variables|UserOptions|Events|GetOn%20Destroyed) 0 (Utilities|Struct|MakeGameplayEventDestroyed (Variables|Default|GetDestroyInstigator)))))

(event Custom|Play25%25Effect
  (bind _damageparticlesystem (Variables|Default|GetDamageParticleSystem))
  (Components|Activation|SetActive _damageparticlesystem true false)
  (Effects|Components|ParticleSystem|SetEmitterEnable _damageparticlesystem "Sparks" true))

(event Custom|InstigateDestroy
  (params (Instigator :type Controller%20Object%20Reference))
  ;[1 1 1] Destruction
  (bind _output_get (Variables|Default|SetDestroyInstigator Instigator))
  (if (or (Math|Byte|NotEqual(Byte) (Team|GetActorTeam _output_get) (Team|GetActorTeam self)) (Knobs|Team|IsTeamNone (Variables|Default|GetTeamSetting)))
    (if (not (Variables|BuildingActor|GetbDestroyed))
      (CallFunction|DoDestroy)
      (CallFunction|HandleHealthChangeEvents 0.0 (Variables|Default|GetDestroyInstigator) true)
      (Minigame|Objective|SetTrackedProgress (Variables|Default|GetFortMinigameObjective) 0.0)
      (Actor|DestroyActor)
      (GameplayEvent|BroadcastEvent (Variables|UserOptions|Events|GetOn%20Destroyed) 0 (Utilities|Struct|MakeGameplayEventDestroyed (Variables|Default|GetDestroyInstigator))))))

(event EventOnDeathServer
  (params (Damage :type Float%20%28single-precision%29) (DamageTags :type Gameplay%20Tag%20Container%20Structure%20%28by%20ref%29) (Momentum :type Vector :default "0, 0, 0") (HitInfo :type Hit%20Result%20Structure%20%28by%20ref%29) (InstigatedBy :type Controller%20Object%20Reference) (DamageCauser :type Actor%20Object%20Reference) (EffectContext :type Gameplay%20Effect%20Context%20Handle%20Structure))
  ;[1 1 1] Destruction
  (Variables|Default|SetDestroyInstigator InstigatedBy)
  (CallFunction|DoDestroy)
  (CallFunction|HandleHealthChangeEvents 0.0 (Variables|Default|GetDestroyInstigator) true)
  (CallFunction|SetVisibilityForcedDuringGame "NewEnumerator2")
  (CallFunction|SetCollisionForcedDuringGame "NewEnumerator2")
  (GameplayEvent|BroadcastEvent (Variables|UserOptions|Events|GetOn%20Destroyed) 0 (Utilities|Struct|MakeGameplayEventDestroyed (Variables|Default|GetDestroyInstigator))))

(event Custom|PickMesh
  (bind _meshindex (Variables|Default|GetMeshIndex))
  (bind _staticmeshes (Variables|Default|GetStaticMeshes))
  (bind _output (Utilities|Array|Get(aref) _staticmeshes _meshindex))
  (bind (_mesh _scale _stattag) (Utilities|Struct|BreakCreativeDeviceMeshSettings _output))
  (if (Utilities|Array|IsValidIndex (Variables|Default|GetStaticMeshes) (Variables|Default|GetMeshIndex))
    (Utilities|FlowControl|Sequence
      (:then_0
        (bind _loaded_object (Creative|Loading|AsyncLoadEditorSafe _mesh)
          (:Loaded
            (Utilities|ValidateAsyncLoadedObject _mesh _loaded_object
              (:true
                (bind _asstatic_mesh (Utilities|Casting|CastToStaticMesh _loaded_object))
                (bind _output_get (Variables|Mesh|SetStaticMesh _asstatic_mesh))
                (Components|StaticMesh|SetStaticMesh (Variables|Debug|GetStaticMeshComponent) _output_get)
                (CallFunction|SetDamageParticlesRelativeLocation)
                (CallFunction|UpdateHealthbarHeight))))))
      (:then_1
        (Variables|Default|SetDestructionTag _stattag)))))

(event Custom|DestroySurroundingActors
  ;[1 1 1] Destruction
  (bind _returnvalue (Utility|OverlapQueue|GetAsyncOverlapQueue self))
  (Utilities|IsValid _returnvalue
    (:Is%20Valid
      (bind _outxy (Fortnite|EvaluateCurveTableRow (Variables|Default|GetCurveTableForDestructionRadius) (Math|Conversions|ToFloat(Integer) (Variables|Default|GetExplosionRadius))))
      (Class|AsyncOverlapQueue|ThrottledAsyncSphereOverlapActors :self _returnvalue :Location (Transformation|GetActorLocation) :Radius _outxy :bCheckLineOfSight false :ShouldCancelProcess (EventDispatchers|CreateEvent :!event ShouldCancelDestruction) :ProcessHitActorFunction (delegate Custom|DestroyHitActor) :CompletionFunction (delegate Custom|OverlapFinished) :ObjectTypes "ObjectTypeQuery1" :TraceChannel "ECC_Visibility")
      ;;[1 1 1] Wait for our destruction to be done.
      (Actor|SetLifeSpan 0.0))))

(event OnAnyPropertyChanged(ToyOptionsComponent)
  ;[0.25 0.5 0.5] Property changes
  (Utilities|FlowControl|Sequence
    (:then_0
      (Gameplay|ForceBuildingHealth (Math|Float|Max(Float) (Variables|Default|GetHealthValue) 1.0))
      (Default|CallUpdateRefresh)
      (CallFunction|UpdateHealthbarHeight)
      (CallFunction|SetShouldTickHealthBarComponent)
      (Class|CreativeUIBroadcastComponent|Initialize (Variables|Default|GetCreative_HUDBroadcast_Component) (Variables|Default|GetFortMinigameProgress) (Variables|UserOptions|Advanced|GetDisplayScoreUpdateOnHUD)))
    (:then_1
      (CallFunction|UpdateTeam))
    (:then_2
      (CallFunction|UpdateComponentVisibility))
    (:then_3
      (Default|UpdateStaticMeshCollision))))

(event OnPropertiesLoaded(ToyOptionsComponent)
  ;[0.25 0.5 0.5] Property changes
  (Utilities|FlowControl|Sequence
    (:then_0
      (Gameplay|ForceBuildingHealth (Math|Float|Max(Float) (Variables|Default|GetHealthValue) 1.0))
      (Default|CallUpdateRefresh)
      (CallFunction|UpdateHealthbarHeight)
      (CallFunction|SetShouldTickHealthBarComponent)
      (Class|CreativeUIBroadcastComponent|Initialize (Variables|Default|GetCreative_HUDBroadcast_Component) (Variables|Default|GetFortMinigameProgress) (Variables|UserOptions|Advanced|GetDisplayScoreUpdateOnHUD)))
    (:then_1
      (CallFunction|UpdateTeam))
    (:then_2
      (CallFunction|UpdateComponentVisibility))
    (:then_3
      (Default|UpdateStaticMeshCollision))))

(event Custom|DestroyHitActor
  (params (HitActor :type Actor%20Object%20Reference))
  ;[1 1 1] Destruction
  (bind _asbuilding_actor (Utilities|Casting|CastToBuildingActor HitActor))
  (if (Variables|/Script/Engine.Actor|GetbCanBeDamaged _asbuilding_actor)
    (bind _asbp_destruction_object (Utilities|Casting|CastToBP_DestructionObject _asbuilding_actor)
      (:then
        (CallFunction|InstigateDestroy _asbp_destruction_object (Variables|Default|GetDestroyInstigator)))
      (:CastFailed
        (Actor|DestroyActor _asbuilding_actor)))))

(event Custom|OverlapFinished
  (params (WasCancelled :type Boolean))
  ;[1 1 1] Destruction
  (Actor|SetLifeSpan 0.01))

(event Custom|UpdateTeam
  (bind _returnvalue (Utilities|GetTeamfromTeamSetting))
  (if (!= (Math|Conversions|ToInteger(Byte) (Variables|Team|GetTeamIndex)) _returnvalue)
    (CallFunction|SetTeam (Math|Conversions|ToByte(Integer) _returnvalue))
    (else
      _))
  (CallFunction|SetParticle))

(event MinigameStartedDelegate(FortMinigameObjective)
  (Networking|SwitchHasAuthority
    (:Authority
      (CallFunction|UpdateTrackingState (Variables|/Script/FortniteGame.FortMinigameLogicComponent|GetCurrentMinigame (Variables|Default|GetFortMinigameObjective))))))

(event MinigameEndedDelegate(FortMinigameObjective)
  (Networking|SwitchHasAuthority
    (:Authority
      (CallFunction|UpdateTrackingState (Variables|/Script/FortniteGame.FortMinigameLogicComponent|GetCurrentMinigame (Variables|Default|GetFortMinigameObjective))))))

(event EventPostFinishSpawning
  (|CallParentFunction)
  (Utilities|FlowControl|Sequence
    (:then_0
      (CallFunction|TeamSettingUpgradePath)
      (CallFunction|UpdateTeam))
    (:then_1
      (CallFunction|PickMesh)
      (Class|FortMinigameLogicComponent|TryAttachtoMinigame (Variables|Default|GetFortMinigameObjective)))))

(event Custom|Notify%20Feedback%20Event
  (params (EventParams :type Gameplay%20Feedback%20Event%20Params%20Structure) (ProgressPct :type Float%20%28double-precision%29))
  (bind (_eventtag _instigatorplayerid) (Utilities|Struct|BreakGameplayFeedbackEventParams EventParams))
  (CallFunction|PostObjectiveFeedbackEvent _eventtag _instigatorplayerid ProgressPct))

(event OnAddedtoMinigame(FortMinigameObjective)
  (params (Minigame :type Fort%20Minigame%20Object%20Reference))
  (Networking|SwitchHasAuthority
    (:Authority
      (CallFunction|UpdateTrackingState Minigame))
    (:Remote
      (Minigame|BindEventtoPlayerAddedDelegate Minigame (EventDispatchers|CreateEvent :!event Custom|OnPlayerAdded))
      (Minigame|BindEventtoPlayerRemovedDelegate Minigame (EventDispatchers|CreateEvent :!event Custom|OnPlayerRemoved))
      (CallFunction|BindAllCurrentMinigamePlayers))))

(event Custom|SetParticle
  (bind _beaconparticletype (Variables|Default|GetBeaconParticleType))
  (bind _beaconparticletypes (Variables|Default|GetBeaconParticleTypes))
  (bind _output (Utilities|Array|Get(aref) _beaconparticletypes _beaconparticletype))
  (bind (_particletype_3_45bbedad49901d6a2ff9fab0205b3f51 _colorparameters_13_112d79bd4c25b6efa3483f992bd13189) (Utilities|Struct|BreakFDestructionObjectParticleType _output))
  (bind _particlesystem (Variables|Default|GetParticleSystem))
  (bind _particlesystem_2 (Variables|Default|GetParticleSystem))
  (Utilities|IsValid _particlesystem_2
    (:Is%20Valid
      (Components|Activation|SetActive _particlesystem_2 false true))
    (:Is%20Not%20Valid
      _))
  (if (Utilities|Array|IsValidIndex _beaconparticletypes _beaconparticletype)
    (bind _object (Utilities|AsyncLoadAsset _particletype_3_45bbedad49901d6a2ff9fab0205b3f51)
      (:Completed
        (Utilities|ValidateAsyncLoadedObject _particletype_3_45bbedad49901d6a2ff9fab0205b3f51 _object
          (:true
            (bind _asparticle_system (Utilities|Casting|CastToParticleSystem _object))
            (Effects|Components|ParticleSystem|SetTemplate (Variables|Default|GetParticleSystem) _asparticle_system)
            (Components|Activation|SetActive _particlesystem true true)
            (CallFunction|SetBeaconColor _particlesystem _colorparameters_13_112d79bd4c25b6efa3483f992bd13189)
            (CallFunction|UpdateHealthandBeaconLocation)
            (CallFunction|SetBeaconScale _particlesystem)
            (if (== (Variables|Default|GetBeaconParticleType) 0)
              (Effects|Components|ParticleSystem|SetEmitterEnable _particlesystem "Letter" false)
              (CallFunction|SetDamageParticlesRelativeLocation))))))))

(event OnRemovedfromMinigame(FortMinigameObjective)
  (params (Minigame :type Fort%20Minigame%20Object%20Reference))
  (Networking|SwitchHasAuthority
    (:Authority
      (Utilities|IsValid Minigame
        (:Is%20Valid
          (Minigame|Objective|StopTracking (Variables|Default|GetFortMinigameObjective) false (Health|GetHealthPercent)))))
    (:Remote
      (Minigame|UnbindEventfromPlayerAddedDelegate Minigame (EventDispatchers|CreateEvent :!event Custom|OnPlayerAdded))
      (Minigame|UnbindEventfromPlayerRemovedDelegate Minigame (EventDispatchers|CreateEvent :!event Custom|OnPlayerRemoved))
      (CallFunction|UnbindAllCurrentMinigamePlayers))))

(event Custom|OnPlayerAdded
  (params (UniqueNetId :type Unique%20Net%20Id%20Structure) (bIsLocalPlayer :type Boolean))
  (Networking|SwitchHasAuthority
    (:Remote
      (bind _returnvalue (UniqueNetID|IsValidNetID UniqueNetId))
      (if (and _returnvalue bIsLocalPlayer)
        (CallFunction|BindPlayer UniqueNetId)))))

(event Custom|OnPlayerRemoved
  (params (UniqueNetId :type Unique%20Net%20Id%20Structure) (bIsLocalPlayer :type Boolean))
  (Networking|SwitchHasAuthority
    (:Remote
      (bind _returnvalue (UniqueNetID|IsValidNetID UniqueNetId))
      (if (and _returnvalue bIsLocalPlayer)
        (CallFunction|UnbindPlayer UniqueNetId)))))

(fn ActivateFromOtherDevice
  (params (Instigator :type Controller%20Object%20Reference) (Requester :type ERequesterType%20Enum :default "NewEnumerator0"))
  (return false))

(fn BindAllCurrentMinigamePlayers
  (bind _fortminigameobjective (Variables|Default|GetFortMinigameObjective))
  (bind _currentminigame (Variables|/Script/FortniteGame.FortMinigameLogicComponent|GetCurrentMinigame _fortminigameobjective))
  (Utilities|IsValid _currentminigame
    (:Is%20Valid
      (bind _output_get (LocalVariables|SetParticipatingPlayers (Minigame|GetParticipatingPlayers _currentminigame)))
      (for _array_element _output_get
        (bind _asfort_player_controller_athena (Utilities|Casting|CastToFortPlayerControllerAthena (Fortnite|GetFortPlayerControllerfromActor _array_element)))
        (bind _output_get_2 (LocalVariables|SetCurrentParticipatingPlayer _asfort_player_controller_athena))
        (if (Pawn|IsLocalPlayerController _output_get_2)
          (FortPlayerControllerAthena|BindEventtoOnLocalPlayerChangedTeams _output_get_2 (EventDispatchers|CreateEvent :!event OnLocalPlayerChangedTeams)))))))

(fn BindPlayer
  (params (UniqueNetID :type Unique%20Net%20Id%20Structure%20%28by%20ref%29))
  (bind _output_get (LocalVariables|SetLocalPlayerControllers (Fortnite|GetLocalFortPlayerControllers true)))
  (bind _array_element (Utilities|Array|ForEachLoopwithBreak _output_get)
    (:LoopBody
      (bind _asfort_player_controller_athena (Utilities|Casting|CastToFortPlayerControllerAthena _array_element))
      (bind _output_get_2 (LocalVariables|SetCurrentLocalPlayer _asfort_player_controller_athena))
      (if (Online|UniqueNetIdRepl|Equal(UniqueIdRepl) (Default|GetGameAccountId _output_get_2) (Variables|GetUniqueNetID))
        (FortPlayerControllerAthena|BindEventtoOnLocalPlayerChangedTeams (LocalVariables|GetCurrentLocalPlayer) (EventDispatchers|CreateEvent :!event OnLocalPlayerChangedTeams))))))

(fn BlueprintModifyIncomingDamage
  (params (Damage :type Float%20%28single-precision%29) (InTags :type Gameplay%20Tag%20Container%20Structure%20%28by%20ref%29) (EffectContext :type Gameplay%20Effect%20Context%20Handle%20Structure%20%28by%20ref%29) (EventInstigator :type Controller%20Object%20Reference) (DamageCauser :type Actor%20Object%20Reference))
  (bind _currentminigame (Variables|/Script/FortniteGame.FortMinigameLogicComponent|GetCurrentMinigame (Variables|Default|GetFortMinigameObjective))
    (:then
      (if (Math|Boolean|ANDBoolean (Minigame|IsinProgress _currentminigame) (Math|Byte|NotEqual(Byte) (Team|GetActorTeam EventInstigator) (Variables|Team|GetTeamIndex)) (or (CallFunction|CanTakeDamage) (GameplayTags|HasTag InTags (LocalVariables|Getlocal%20Indestructible%20Damage%20Tag) true)))
        (return Damage)
        (else
          _)))
    (:else
      _))
  (return 0.0))

(fn CanInteract
  (params (InteractingPawn :type Fort%20Pawn%20Object%20Reference) (InteractionBeingAttempted :type EInteractionBeingAttempted%20Enum :default "FirstInteraction") (InteractionType :type TInteractionType%20Enum :default "IT_NoInteraction"))
  (return false))

(fn CanTakeDamage
  (return (> (Variables|Default|GetHealthValue) 0.0)))

(fn DoDestroy
  (bind _currentminigame (Variables|/Script/FortniteGame.FortMinigameLogicComponent|GetCurrentMinigame (Variables|Default|GetFortMinigameObjective))
    (:then
      (Minigame|IsInProgress _currentminigame
        (:Yes
          (CallFunction|SendDestroyedEvent)
          (if (Utilities|Enum|Equal(Enum) (Variables|UserKnobs|GetPlayAudioVFXKnob) "NewEnumerator1")
            (CallFunction|RPCPlayDeathSound))
          (if (> (Variables|Default|GetExplosionRadius) 0)
            (Environment|ApplyGlobalEnvironmentGameplayEffecttoActor self "/Game/Creative/Devices/Destruction_Object/GE_Creative_DestructionObject.GE_Creative_DestructionObject_C")
            (CallFunction|DestroySurroundingActors)))))
    (:else)))

(fn GetDeviceColor
  (bind _output (CallFunction|GetBeaconColor))
  (return (Math|Conversions|ToLinearColor(Vector) _output)))

(fn GetInstigatingController
  (return 0))

(fn GetBeaconColor
  (Utilities|FlowControl|Sequence
    (:then_0
      (if (!= (Variables|Default|GetBeaconColorIndex) -1)
        (LocalVariables|SetLocalIndex (Variables|Default|GetBeaconColorIndex))
        (else
          (LocalVariables|SetLocalIndex (Utilities|Select :Index (Team|GetTeamAffiliationtoLocalPlayer (Math|Conversions|ToByte(Integer) (Utilities|GetTeamfromTeamSetting)) true true) :Friendly (Variables|Default|GetFriendlyTeamBeaconColorIndex) :Neutral (Variables|Default|GetFriendlyTeamBeaconColorIndex) :Hostile (Variables|Default|GetEnemyTeamBeaconColorIndex))))))
    (:then_1
      (if (Utilities|Array|IsValidIndex (Variables|Default|GetBeaconColors) (LocalVariables|GetLocalIndex))
        (return (Utilities|Array|Get(acopy) (Variables|Default|GetBeaconColors) (LocalVariables|GetLocalIndex)) true)
        (else
          (return "0, 0, 0" false))))))

(fn GetExplosionFXScale
  (return (Utilities|Select :Index (Math|Integer|Min(Integer) (Variables|Default|GetExplosionRadius) 4) :Option%200 "0.250000,0.250000,0.250000" :Option%201 "0.250000,0.250000,0.250000" :Option%202 "0.500000,0.500000,0.500000" :Option%203 "1.000000,1.000000,1.000000" :Option%204 "2.000000,2.000000,2.000000")))

(fn GetObjectiveNameStyle
  (params (EventTag :type Gameplay%20Tag%20Structure))
  (bind _teamindex (Variables|Team|GetTeamIndex))
  (bind _returnvalue (Team|GetTeamAffiliationtoLocalPlayer _teamindex true true))
  (switch Utilities|FlowControl|Switch|SwitchonGameplayTag EventTag
    (:Default
      _)
    (:Creative.Event.Objective.Warning
      (return (Utilities|Select :Index _returnvalue :Friendly "NSLOCTEXT(\"[7DC33254425099E42CE0DD9365D86844]\", \"B421D1004E0EC1D6BDB23FA59E7F80E4\", \"Ally\")" :Neutral "NSLOCTEXT(\"[7DC33254425099E42CE0DD9365D86844]\", \"9297A007466469B693432F836BE514F7\", \"Default\")" :Hostile "NSLOCTEXT(\"[7DC33254425099E42CE0DD9365D86844]\", \"4628D41F4E48B3C147B76393B6D4451E\", \"Enemy\")")))
    (:Creative.Event.Objective.Critical
      _)
    (:Creative.Event.Objective.Destroyed
      _))
  (return (Utilities|Select :Index _returnvalue :Friendly "NSLOCTEXT(\"[7DC33254425099E42CE0DD9365D86844]\", \"1CB0F6D64A192F2FA378BBBD0E50B676\", \"TeamVictim\")" :Neutral "NSLOCTEXT(\"[7DC33254425099E42CE0DD9365D86844]\", \"2BF3E2E848DA8DD42304C3A560C746B2\", \"Victim\")" :Hostile "NSLOCTEXT(\"[7DC33254425099E42CE0DD9365D86844]\", \"FB77F75D4784358FCE1C6B9F0C2E4713\", \"Victim\")")))

(fn GetOverrideMeshMaterial
  (params (MeshComp :type Static%20Mesh%20Component%20Object%20Reference) (MatIdx :type Integer))
  (return (Rendering|Material|GetMaterial MeshComp MatIdx)))

(fn GetSpawnedPickups
  (return 0))

(fn GetUIHideCondition
  (bind _showhealthbar (Variables|UserKnobs|GetShowHealthBar))
  (if (> _showhealthbar 0.0)
    (return "NewEnumerator2")
    (elif (Math|Float|NearlyEqual(Float) _showhealthbar -1.0 0.1)
      (return "NewEnumerator0")
      (else
        (return "NewEnumerator1")))))

(fn GetWidgetOwnerLocation
  (params (WidgetComponent :type Widget%20Component%20Object%20Reference))
  (Utilities|FlowControl|Sequence
    (:then_0
      (Utilities|IsValid WidgetComponent
        (:Is%20Valid
          (bind _output_get (LocalVariables|SetWidgetObject (UserInterface|GetUserWidgetObject (Variables|GetWidgetComponent))))
          (Utilities|IsValid _output_get
            (:Is%20Valid
              (bind _output_get_2 (LocalVariables|SetPlayer (Widget|GetOwningPlayer _output_get)))
              (Utilities|IsValid _output_get_2
                (:Is%20Valid
                  (bind _output_get_3 (LocalVariables|SetControlledPawn (Pawn|GetControlledPawn _output_get_2)))
                  (Utilities|IsValid _output_get_3
                    (:Is%20Valid
                      (LocalVariables|SetPlayerLocation (Transformation|GetActorLocation _output_get_3)))))))))))
    (:then_1
      (return (LocalVariables|GetPlayerLocation)))))

(fn HandleHealthChangeEvents
  (params (HealthPct :type Float%20%28double-precision%29) (Instigator :type Controller%20Object%20Reference) (bProcessDestroy :type Boolean))
  (Utilities|FlowControl|Sequence
    (:then_0
      (LocalVariables|SetPrevProgressPct (Variables|/Script/FortniteGame.FortMinigameObjectiveComponent|GetProgressPercentage (Variables|Default|GetFortMinigameObjective)))
      (bind _instigator (Variables|Actor|GetInstigator)
        (:then
          (bind _asfort_player_state (Utilities|Casting|CastToFortPlayerState (Variables|/Script/Engine.Controller|GetPlayerState _instigator)))
          (LocalVariables|SetInstigatorPlayerId (PlayerState|GetUniqueID _asfort_player_state)))
        (:else)))
    (:then_1
      (if (Math|Boolean|ANDBoolean (Variables|Default|GetbShowDestroyEvents) bProcessDestroy (<= (Variables|GetHealthPct) 0.0))
        (CallFunction|SendObjectiveFeedbackEvent "(TagName=\"Creative.Event.Objective.Destroyed\")" (LocalVariables|GetInstigatorPlayerId))
        (elif (and (> (LocalVariables|GetPrevProgressPct) (Variables|Default|GetCriticalEventPct)) (<= (Variables|GetHealthPct) (Variables|Default|GetCriticalEventPct)))
          (CallFunction|SendObjectiveFeedbackEvent "(TagName=\"Creative.Event.Objective.Critical\")" (LocalVariables|GetInstigatorPlayerId) (Variables|Default|GetCriticalEventPct))
          (elif (and (> (LocalVariables|GetPrevProgressPct) (Variables|Default|GetWarningEventPct)) (<= (Variables|GetHealthPct) (Variables|Default|GetWarningEventPct)))
            (CallFunction|SendObjectiveFeedbackEvent "(TagName=\"Creative.Event.Objective.Warning\")" (LocalVariables|GetInstigatorPlayerId) (Variables|Default|GetWarningEventPct))))))))

(fn IsEnabled
  (return false))

(fn IsHoldingItemForPawn
  (params (Item :type Item%20Variant%20Handle%20Structure) (Pawn :type Pawn%20Object%20Reference))
  (return false))

(fn IsInPlayMode
  (bind _fortminigamelogic (Variables|Default|GetFortMinigameLogic)
    (:then
      (bind _currentminigame (Variables|/Script/FortniteGame.FortMinigameLogicComponent|GetCurrentMinigame _fortminigamelogic)
        (:then
          (return (Minigame|IsinPlayMode _currentminigame :!class /Script/FortniteGame.FortMinigame)))
        (:else
          _)))
    (:else
      _))
  (return false))

(fn IsTargetableByLinearMover
  (return true))

(fn MessageScore
  (return (Variables|UserOptions|Advanced|GetHUD%20Message) (Variables|UserOptions|Advanced|GetHUD%20MessageScoreColor) (Variables|UserOptions|Advanced|GetHUD%20MessageColor) (Variables|UserOptions|Advanced|GetResetHUD%20MessageScore)))

(fn OnRep_BeaconColorIndex
  (CallFunction|SetParticle))

(fn OnRep_BeaconParticleType
  (CallFunction|SetParticle))

(fn OnRep_BeaconScale
  (CallFunction|SetBeaconScale (Variables|Default|GetParticleSystem)))

(fn OnRep_ClampToScreen)

(fn OnRep_CollisionForcedDuringGame
  (Default|UpdateStaticMeshCollision))

(fn OnRep_MeshIndex
  (CallFunction|PickMesh))

(fn OnRep_NewTeamIndex
  (CallFunction|SetParticle)
  (Default|CallLocalPlayerChangedTeams))

(fn PostObjectiveFeedbackEvent
  (params (EventTag :type Gameplay%20Tag%20Structure) (InstigatorPlayerId :type Unique%20Net%20Id%20Structure) (ProgressPct :type Float%20%28double-precision%29))
  (bind _feedbackeventmessages (Variables|UI|GetFeedbackEventMessages))
  (bind (_value _returnvalue) (Utilities|Map|Find _feedbackeventmessages EventTag))
  (if _returnvalue
    (FeedbackEvents|PostFeedbackEventfromGameMessage self _value (Variables|GetInstigatorPlayerId) (Utilities|Map|MakeMap "ProgressPct" (Utilities|Text|FormatText "NSLOCTEXT(\"[7DC33254425099E42CE0DD9365D86844]\", \"6268DD6643843B15BD7D1EA2AEDE11A4\", \"{ProgressPct}%\")" (Math|Float|Truncate (* ProgressPct 1e+02))) "ObjectiveStyle" (CallFunction|GetObjectiveNameStyle (Variables|GetEventTag)) "ObjectiveName" (select (Utilities|String|IsEmpty (Variables|UserKnobs|GetCustomObjective)) (CallFunction|GetObjectiveNameText (Utilities|Array|FindItem (Variables|Default|GetTextureArrayForIndicator) (Utilities|Struct|BreakCreativeIconOption (Variables|UserKnobs|GetIndicatorPresetIndex)))) (Utilities|Text|FormatText "NSLOCTEXT(\"[7DC33254425099E42CE0DD9365D86844]\", \"F89AB9684BB15F7FC4BBEF92BE9C98FF\", \"Objective {CustomObjective}\")" (Variables|UserKnobs|GetCustomObjective)))))))

(fn SendDestroyedEvent
  (bind (_outsourcetags _outcontexttags) (Class|FortMinigameStatsComponent|GetSourceandContextTags (LocalVariables|Getlocal%20MinigameStatsComp) 0))
  (bind (_outsourcetags_2 _outcontexttags_2) (Class|FortMinigameStatsComponent|GetSourceandContextTags (LocalVariables|Getlocal%20MinigameStatsComp) (LocalVariables|Getlocal%20FortPlayerController)))
  (bind _asfort_player_controller (Utilities|Casting|CastToFortPlayerController (Variables|Default|GetDestroyInstigator)))
  (LocalVariables|Setlocal%20FortPlayerController _asfort_player_controller)
  (bind _local_fortplayercontroller (LocalVariables|Getlocal%20FortPlayerController)
    (:then
      (bind _returnvalue (CreativeDevices|GetMinigameStatsComp _local_fortplayercontroller))
      (LocalVariables|Setlocal%20MinigameStatsComp _returnvalue)
      (bind _local_minigamestatscomp (LocalVariables|Getlocal%20MinigameStatsComp)
        (:then
          (Class|FortMinigameStatsComponent|SendStatEventwithTags _local_minigamestatscomp (LocalVariables|Getlocal%20FortPlayerController) "MinigameDynamicEvent" self (GameplayTags|MakeGameplayTagContainerfromTag (Variables|Default|GetDestructionTag)) _outsourcetags_2 _outcontexttags_2 1)
          (Class|FortMinigameStatsComponent|SendStatEventwithTags (LocalVariables|Getlocal%20MinigameStatsComp) (LocalVariables|Getlocal%20FortPlayerController) "MinigameDynamicEvent" self (GameplayTags|MakeGameplayTagContainerfromTag "(TagName=\"Minigame.Stat.Score\")") _outsourcetags _outcontexttags (Variables|UserKnobs|GetScoreValueKnob))
          (Class|CreativeHUDBroadcastComponent|DoScoreEvent (Variables|Default|GetCreative_HUDBroadcast_Component) (LocalVariables|Getlocal%20FortPlayerController) (Variables|UserKnobs|GetScoreValueKnob) (Variables|UserOptions|Advanced|GetDisplayScoreUpdateOnHUD)))
        (:else)))
    (:else)))

(fn SendObjectiveFeedbackEvent
  (params (EventTag :type Gameplay%20Tag%20Structure) (PlayerId :type Unique%20Net%20Id%20Structure) (ProgressPct :type Float%20%28double-precision%29))
  (CallFunction|NotifyFeedbackEvent (Utilities|Struct|MakeGameplayFeedbackEventParams EventTag PlayerId) ProgressPct))

(fn SetDamageParticlesRelativeLocation
  (bind _fxdamagesocketname (Variables|Default|GetFxDamageSocketName))
  (bind _staticmeshcomponent (Variables|Debug|GetStaticMeshComponent))
  (bind (_returnvalue_location _returnvalue_rotation _returnvalue_scale) (Math|Transform|BreakTransform (Transformation|GetSocketTransform _staticmeshcomponent _fxdamagesocketname "RTS_Component")))
  (bind (_returnvalue_scale_x _returnvalue_scale_y _returnvalue_scale_z) (Math|Vector|BreakVector _returnvalue_scale))
  (bind _returnvalue (Math|Float|Max(Float) _returnvalue_scale_x _returnvalue_scale_y _returnvalue_scale_z))
  (bind _damageparticlesystem (Variables|Default|GetDamageParticleSystem))
  (if (Transformation|DoesSocketExist (Variables|Debug|GetStaticMeshComponent) (Variables|Default|GetFxDamageSocketName))
    (Transformation|SetRelativeLocationAndRotation (Variables|Default|GetDamageParticleSystem) _returnvalue_location _returnvalue_rotation false false)
    (Effects|Components|ParticleSystem|SetFloatParameter _damageparticlesystem "Radius" _returnvalue)
    (Effects|Components|ParticleSystem|SetFloatParameter _damageparticlesystem "SpawnRateScale" (/ (* (Math|Trig|GetPI) (Math|Float|Square _returnvalue)) 7853.97998))))

(fn SetTriggerCollision
  (params (On :type Boolean))
  (return false))

(fn SetBeaconColor
  (params (Beacon :type Cascade%20Particle%20System%20Component%20Object%20Reference) (Array :type Array%20of%20FDestruction%20Object%20Particle%20Color%20Parameter%20Structures))
  (Utilities|IsValid Beacon
    (:Is%20Valid
      (bind (_output _bisvalid) (CallFunction|GetBeaconColor))
      (LocalVariables|SetNewBeaconColor _output)
      (if _bisvalid
        (for _array_element Array
          (Effects|Components|ParticleSystem|SetVectorParameter (Variables|GetBeacon) (Utilities|Struct|BreakFDestructionObjectParticleColorParameter _array_element) (* (LocalVariables|GetNewBeaconColor) (.ColorParameterMultiplier_7_AE2CF29447670F4CE0F86BA17A6BB6D6 (Utilities|Struct|BreakFDestructionObjectParticleColorParameter _array_element)))))))))

(fn SetBeaconScale
  (params (BeaconParticleEffect :type Scene%20Component%20Object%20Reference))
  (bind _beaconscale (Variables|Default|GetBeaconScale))
  (Utilities|IsValid BeaconParticleEffect
    (:Is%20Valid
      (if (> _beaconscale 0.0)
        (Transformation|SetWorldScale3D BeaconParticleEffect (Math|Conversions|ToVector(Float) _beaconscale))
        (else
          (Transformation|SetRelativeScale3D BeaconParticleEffect "1.000000,1.000000,1.000000"))))))

(fn SetShouldTickHealthBarComponent
  (bind _showhealthbar (Variables|UserKnobs|GetShowHealthBar))
  (bind _returnvalue (<= _showhealthbar 0.0))
  (if _returnvalue
    (Components|Tick|SetComponentTickEnabled (Variables|Default|GetHealthBarComp) false)
    (else
      (Components|Tick|SetComponentTickEnabled (Variables|Default|GetHealthBarComp) true)))
  (bind _asobjective_basic_health_bar (Utilities|Casting|CastToObjectiveBasicHealthBar (UserInterface|GetUserWidgetObject (Variables|Default|GetHealthBarComp))))
  (Class|ObjectiveBasicHealthBar|UpdateVisibilityTimer _asobjective_basic_health_bar (not _returnvalue)))

(fn ShouldCancelDestruction
  (bind _self self)
  (bind _returnvalue (Minigame|GetMinigame _self :!class /Script/FortniteGame.MinigameBlueprintLibrary))
  (Utilities|IsValid _returnvalue
    (:Is%20Valid
      (return (not (Minigame|IsinProgress _returnvalue))))
    (:Is%20Not%20Valid
      (return true))))

(fn ShouldInvisiblePropsBeAllowed
  (return false))

(fn ShouldShowFX
  (return (and (Utilities|Enum|Equal(Enum) (Variables|UserKnobs|GetPlayAudioVFXKnob) "NewEnumerator1") (Rendering|IsVisible (Variables|Debug|GetStaticMeshComponent)))))

(fn TeamSettingUpgradePath
  (bind _toyoptionscomponent (Variables|Default|GetToyOptionsComponent))
  (bind (_currentoptions _returnvalue) (OptionData|GetPropertyOverrides _toyoptionscomponent))
  (bind (_value _returnvalue_2) (Utilities|Map|Find _currentoptions "NewTeamIndex"))
  (bind _toyoptionscomponent_2 (Variables|Default|GetToyOptionsComponent))
  (Networking|SwitchHasAuthority
    (:Authority
      (if _returnvalue
        (if _returnvalue_2
          (bind _output_get (LocalVariables|SetNewTeamIndexPropertyOverride _value))
          (bind _output_get_2 (LocalVariables|SetActualTeamIndexOverride (Utilities|String|StringToInteger _output_get)))
          (OptionData|SetPropertyOverrides _toyoptionscomponent_2 (Utilities|Map|MakeMap (Class|PlaylistUserOptionBase|GetOptionKey (Variables|Default|GetTeamSettingOption) false 0) (Class|PlaylistUserOptionBase|GetOptionValueNamefromValue (Variables|Default|GetTeamSettingOption) (Utilities|String|ToString(Integer) (Utilities|Select :Index _output_get_2 :Option%200 15 :Option%201 -1 :Option%202 -1 :Option%203 0 :Option%204 1 :Option%205 2 :Option%206 3 :Option%207 4 :Option%208 5 :Option%209 6 :Option%2010 7 :Option%2011 8 :Option%2012 9 :Option%2013 10 :Option%2014 11 :Option%2015 12 :Option%2016 13 :Option%2017 14)) true 0)))
          (OptionData|RemovePropertyOverrides _toyoptionscomponent_2 (Utilities|Array|MakeArray "NewTeamIndex")))))))

(fn UnbindAllCurrentMinigamePlayers
  (bind _fortminigameobjective (Variables|Default|GetFortMinigameObjective))
  (bind _currentminigame (Variables|/Script/FortniteGame.FortMinigameLogicComponent|GetCurrentMinigame _fortminigameobjective))
  (Utilities|IsValid _currentminigame
    (:Is%20Valid
      (bind _output_get (LocalVariables|SetParticipatingPlayers (Minigame|GetParticipatingPlayers _currentminigame)))
      (for _array_element _output_get
        (bind _asfort_player_controller_athena (Utilities|Casting|CastToFortPlayerControllerAthena (Fortnite|GetFortPlayerControllerfromActor _array_element)))
        (bind _output_get_2 (LocalVariables|SetCurrentParticipatingPlayer _asfort_player_controller_athena))
        (if (Pawn|IsLocalPlayerController _output_get_2)
          (FortPlayerControllerAthena|UnbindEventfromOnLocalPlayerChangedTeams _output_get_2 (EventDispatchers|CreateEvent :!event OnLocalPlayerChangedTeams)))))))

(fn UnbindPlayer
  (params (UniqueNetID :type Unique%20Net%20Id%20Structure%20%28by%20ref%29))
  (bind _output_get (LocalVariables|SetLocalPlayerControllers (Fortnite|GetLocalFortPlayerControllers true)))
  (bind _array_element (Utilities|Array|ForEachLoopwithBreak _output_get)
    (:LoopBody
      (bind _asfort_player_controller_athena (Utilities|Casting|CastToFortPlayerControllerAthena _array_element))
      (bind _output_get_2 (LocalVariables|SetCurrentLocalPlayer _asfort_player_controller_athena))
      (if (Online|UniqueNetIdRepl|Equal(UniqueIdRepl) (Default|GetGameAccountId _output_get_2) (Variables|GetUniqueNetID))
        (FortPlayerControllerAthena|UnbindEventfromOnLocalPlayerChangedTeams (LocalVariables|GetCurrentLocalPlayer) (EventDispatchers|CreateEvent :!event OnLocalPlayerChangedTeams))))))

(fn UpdateComponentVisibility
  (Visibility|SetVisiblityDuringPhase (Variables|Default|GetVisibleInGameComponent) (select (Variables|UserKnobs|GetIsMeshVisibleDuringMinigame) "NewEnumerator3" "NewEnumerator1")))

(fn UpdateHealthAndBeaconLocation
  (bind (_origin _boxextent) (Collision|GetComponentBounds (Variables|Debug|GetStaticMeshComponent)))
  (Transformation|SetWorldLocationAndRotation (Variables|Default|GetParticleSystem) (Math|Vector|vector+vector _origin (Math|Vector|MakeVector 0.0 0.0 (.z _boxextent))) "0, 0, 0" false false))

(fn UpdateHealthEffects
  (params (NewHealth :type Float%20%28double-precision%29))
  (if (CallFunction|CanTakeDamage)
    (bind _output_get (LocalVariables|SetHealthPercent (/ NewHealth (Variables|Default|GetHealthValue))))
    (if (> _output_get 0.5)
      (if (Variables|Default|GetbDamageEffectsAreActive)
        (CallFunction|StopHealthEffects)
        (Variables|Default|SetbDamageEffectsAreActive false))
      (elif (> (LocalVariables|GetHealthPercent) 0.25)
        (CallFunction|Play50%25Effect)
        (Variables|Default|SetbDamageEffectsAreActive true)
        (elif (> (LocalVariables|GetHealthPercent) 0.0)
          (CallFunction|Play25%25Effect)
          (Variables|Default|SetbDamageEffectsAreActive true)
          (elif (Variables|Default|GetbDamageEffectsAreActive)
            (CallFunction|StopHealthEffects)
            (Variables|Default|SetbDamageEffectsAreActive false)))))))

(fn UpdateStaticMeshCollision
  (bind _collisionprofilename (LocalVariables|GetCollisionProfileName))
  (Utilities|FlowControl|Sequence
    (:then_0
      (LocalVariables|SetNoCollisionProfileName (CreativeDevices|GetNamefromCollisionProfile "(Name=\"FortCreativeOnlyVolumeManager\")"))
      (LocalVariables|SetEditModeCollisionProfileName (CreativeDevices|GetNamefromCollisionProfile "(Name=\"FortDynamicMeshPhysics\")"))
      (LocalVariables|SetPlayModeProfileName (CreativeDevices|GetNamefromCollisionProfile (Variables|UserKnobs|GetCollisionProfileDuringGamesKnob)))
      (if (Default|IsinPlayMode)
        (switch Utilities|FlowControl|Switch|SwitchonEnum_ForcedValueDuringGame (Variables|Replicated|GetCollisionForcedDuringGame)
          (:NewEnumerator0
            (switch Utilities|FlowControl|Switch|SwitchonEnum_CollisionDuringGamesSetting (Variables|UserKnobs|GetCollisionDuringGamesKnob)
              (:NewEnumerator0
                (LocalVariables|SetCollisionProfileName (LocalVariables|GetNoCollisionProfileName))
                (Variables|Actor|SetbCanBeDamaged false))
              (:NewEnumerator1
                (LocalVariables|SetCollisionProfileName (LocalVariables|GetPlayModeProfileName))
                (Variables|Actor|SetbCanBeDamaged true))
              (:NewEnumerator3
                (if (Variables|UserKnobs|GetIsMeshVisibleDuringMinigame)
                  (LocalVariables|SetCollisionProfileName (LocalVariables|GetPlayModeProfileName))
                  (Variables|Actor|SetbCanBeDamaged true)
                  (else
                    (LocalVariables|SetCollisionProfileName (LocalVariables|GetNoCollisionProfileName))
                    (Variables|Actor|SetbCanBeDamaged false))))))
          (:NewEnumerator1
            (LocalVariables|SetCollisionProfileName (LocalVariables|GetPlayModeProfileName))
            (Variables|Actor|SetbCanBeDamaged true))
          (:NewEnumerator2
            (LocalVariables|SetCollisionProfileName (LocalVariables|GetNoCollisionProfileName))
            (Variables|Actor|SetbCanBeDamaged false)))
        (else
          (LocalVariables|SetCollisionProfileName (LocalVariables|GetEditModeCollisionProfileName))
          (Variables|Actor|SetbCanBeDamaged true) ; Default to how everything acts
        )))
    (:then_1
      (if (Utilities|Name|NotEqual(Name) _collisionprofilename "None")
        (Collision|SetCollisionProfileName (Variables|Debug|GetStaticMeshComponent) _collisionprofilename true :!class /Script/Engine.PrimitiveComponent)))))

(fn UpdateTrackedProgress
  (bind _returnvalue (Health|GetHealthPercent))
  (Networking|SwitchHasAuthority
    (:Authority
      (CallFunction|HandleHealthChangeEvents _returnvalue)
      (Minigame|Objective|SetTrackedProgress (Variables|Default|GetFortMinigameObjective) _returnvalue))))

(fn UpdateTrackingState
  (params (Minigame :type Fort%20Minigame%20Object%20Reference))
  (Utilities|IsValid Minigame
    (:Is%20Valid
      (if (Variables|Default|GetbTrackProgress)
        (Minigame|HasStarted (Variables|GetMinigame) :!class /Script/FortniteGame.FortMinigame
          (:Yes
            (if (not (Minigame|IsPostGame (Variables|GetMinigame)))
              (Minigame|Objective|StartTracking (Variables|Default|GetFortMinigameObjective) (- (Utilities|Array|FindItem (Variables|Default|GetTextureArrayForIndicator) (Utilities|Struct|BreakCreativeIconOption (Variables|UserKnobs|GetIndicatorPresetIndex))) 1) (Health|GetHealthPercent) "DestructionObjective" 0 0 true)))
          (:No
            _))
        (Minigame|Objective|StopTracking (Variables|Default|GetFortMinigameObjective) true (Health|GetHealthPercent))
        (else
          _)))))

(fn ConstructionScript)
